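# Material function for the Parallel Alien character skin that applies a masked thin-film effect and a colored directional rim glow to incoming material attributes.
# ThinFilmFunction runs first, using a selectable channel of the SkinFX mask; the rim is then added to Emissive.
# The rim dots pixel normals with a camera-space vector, then is shaped by hardness, exponent and strength.
# Mask red, a pre-skinned height gradient and a fade at vertical view angles limit it.
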MaterialFunction /BRCosmetics/Characters/Player/Male/Medium/Bodies/M_MED_Parallel_Alien/Materials/MF_ParallelAlien
graph {nodes [
  n1 = TextureSample(Tex=NamedRerouteDeclaration(None=TextureObjectParameter() {group=Skin FX, parameterName=SkinFX_Mask}) {name=SkinFX_Mask, nodeColor=<rgba 0.9842 0 1 1>})
  n2 = GetMaterialAttributes(None=MaterialFunctionCall!(
    Alpha=StaticComponentMaskParameter(None=@n1.RGB) {defaultR=true, group=Skin FX, parameterName=ThinFilm_Channel}
    Material Attributes=FunctionInput() {inputName=Material, inputType=FunctionInput_MaterialAttributes, sortPriority=-20}
   ) {materialFunction=/Game/Weapons/WeaponSkins/MaterialFunction/ThinFilmFunction}.Result) {attributeGetTypes [#1, #2]}
  n3 = VectorParameter() {group=Parallel, parameterName=parallel_rimLightColor, sortPriority=0, defaultValue=<rgba 0.0086 0.4508 0.6939 1>}
  n4 = PixelNormalWS()
  n5 = VectorParameter() {group=Parallel, parameterName=parallel_rimLightVector, sortPriority=10, defaultValue=<rgba 1 0.3 1 0>}
  n6 = Transform(None=@n5.RGB) {transformSourceType=TRANSFORMSOURCE_Camera}
  n7 = Power(
   Base=Multiply(A=DotProduct(A=@n4, B=@n6), B=ScalarParameter() {defaultValue=1, group=Parallel, parameterName=parallel_rimHardness, sortPriority=20}) {constB=10}
   Exp=ScalarParameter() {defaultValue=1, group=Parallel, parameterName=parallel_rimExponent}
  )
  n8 = MaterialFunctionCall(Float3=FunctionInput() {inputName=PreSkinnedLocalPosition}) {materialFunction=&BreakOutFloat3Components}
  n9 = Multiply(
   A=Multiply(A=Saturate(None=@n7), B=ScalarParameter() {defaultValue=0.5, group=Parallel, parameterName=parallel_rimStrength})
   B=Saturate(None=InverseLinearInterpolate(Value=@n8.B) {constA=10, constB=150})
  )
  n10 = DotProduct(A=CameraVectorWS(), B=Constant3Vector() {constant=<rgba 0 0 1 0>})
  n11 = Saturate(None=OneMinus(None=Abs(None=@n10)))
  n12 = Multiply(A=@n3.RGB, B=Multiply(A=@n1.R, B=Multiply(A=@n9, B=@n11)))
  FunctionOutput(None=SetMaterialAttributes!(MaterialAttributes=@n2.MaterialAttributes, EmissiveColor=Add(A=@n2.EmissiveColor, B=@n12), BaseColor=@n2.MaterialAttributes) {attributeSetTypes [#1, #2]})
  VertexNormalWS()
 ]}
